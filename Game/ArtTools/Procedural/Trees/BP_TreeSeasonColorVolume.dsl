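; Sphere volume that recolours Parent_Tree_Apollo trees tagged Proc_SeasonTreeColor from its construction script.
; UpdateTrees gives trees within the radius, subject to priority checks, either the volume's colour and variations or the matching preset from each tree's season data.
; A local update also resets trees between one and two radii to their defaults, then re-runs every volume in global mode, which skips resetting, to restore overlapping influence.
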(event EventBeginPlay)

(event Collision|EventActorBeginOverlap
  (params (OtherActor :type Actor%20Object%20Reference)))

(event EventTick
  (params (DeltaSeconds :type Float%20%28single-precision%29)))

(fn UpdateTrees
  (params (IsGlobalUpate :type Boolean))
  (bind _returnvalue (Transformation|GetActorLocation))
  (bind _sphere (Variables|Default|GetSphere))
  (bind _returnvalue_2 (Components|Sphere|GetScaledSphereRadius _sphere))
  (bind _priority (Variables|SeasonColor|GetPriority))
  (Utilities|FlowControl|Sequence
    (:then_0
      (bind _outactors (Actor|GetAllActorswithTag "Proc_SeasonTreeColor"))
      (for _array_element _outactors
        (if (<= (Math|Vector|Distance(Vector) (Transformation|GetActorLocation _array_element) _returnvalue) _returnvalue_2)
          (bind _asparent_tree_apollo (Utilities|Casting|CastToParent_Tree_Apollo _array_element))
          (if (and (<= (Variables|/Game/Building/ActorBlueprints/Prop/Parent_Tree_Apollo.Parent_Tree_Apollo_C|GetPrioritySetbyVolume _asparent_tree_apollo) _priority) (< (Variables|/Game/Building/ActorBlueprints/Prop/Parent_Tree_Apollo.Parent_Tree_Apollo_C|GetPriority _asparent_tree_apollo) _priority))
            (Variables|/Game/Building/ActorBlueprints/Prop/Parent_Tree_Apollo.Parent_Tree_Apollo_C|SetPrioritySetbyVolume _priority _asparent_tree_apollo)
            (if (Utilities|Enum|Equal(Enum) (Variables|SeasonColor|GetTree_Color_Option) "NewEnumerator0")
              (Variables|/Game/Building/ActorBlueprints/Prop/Parent_Tree_Apollo.Parent_Tree_Apollo_C|SetColor (Variables|SeasonColor|GetColor) _asparent_tree_apollo)
              (Variables|/Game/Building/ActorBlueprints/Prop/Parent_Tree_Apollo.Parent_Tree_Apollo_C|SetValueVariation (Variables|SeasonColor|GetValueVariation) _asparent_tree_apollo)
              (Variables|/Game/Building/ActorBlueprints/Prop/Parent_Tree_Apollo.Parent_Tree_Apollo_C|SetHueVariation (Variables|SeasonColor|GetHueVariation) _asparent_tree_apollo)
              (Class|ParentTreeApollo|UpdateColor _asparent_tree_apollo)
              (else
                (bind _array_element_2 (Utilities|Array|ForEachLoopwithBreak (Variables|/Game/ArtTools/Procedural/Trees/TreeSeasonColor_Data.TreeSeasonColor_Data_C|GetTreeSeasonData (Variables|/Game/Building/ActorBlueprints/Prop/Parent_Tree_Apollo.Parent_Tree_Apollo_C|GetTreeSeasonColor%20Data _asparent_tree_apollo)))
                  (:LoopBody
                    (if (Utilities|Enum|Equal(Enum) (Utilities|Struct|BreakTreeSeasonColorStruct _array_element_2) (Variables|SeasonColor|GetTree_Color_Option))
                      (Variables|/Game/Building/ActorBlueprints/Prop/Parent_Tree_Apollo.Parent_Tree_Apollo_C|SetColor (.Color_5_E33006414CF6D999D4D482A4D2FF03F6 (Utilities|Struct|BreakTreeSeasonColorStruct _array_element_2)) _asparent_tree_apollo)
                      (Variables|/Game/Building/ActorBlueprints/Prop/Parent_Tree_Apollo.Parent_Tree_Apollo_C|SetValueVariation (.ValueVariation_9_4FAED9E04DA8955B0BC4EDB1E0A3C1F7 (Utilities|Struct|BreakTreeSeasonColorStruct _array_element_2)) _asparent_tree_apollo)
                      (Variables|/Game/Building/ActorBlueprints/Prop/Parent_Tree_Apollo.Parent_Tree_Apollo_C|SetHueVariation (.HueVariation_10_00BFFB9140C4AD7D16C9AA9AEB01A935 (Utilities|Struct|BreakTreeSeasonColorStruct _array_element_2)) _asparent_tree_apollo)
                      (Class|ParentTreeApollo|UpdateColor _asparent_tree_apollo)))))))
          (elif (<= (Math|Vector|Distance(Vector) (Transformation|GetActorLocation _array_element) _returnvalue) (* _returnvalue_2 2.0))
            (if (not IsGlobalUpate)
              ;[1 1 1] Reset Color to Default
              (bind _asparent_tree_apollo_2 (Utilities|Casting|CastToParent_Tree_Apollo _array_element))
              (if (Utilities|Enum|Equal(Enum) (Variables|/Game/Building/ActorBlueprints/Prop/Parent_Tree_Apollo.Parent_Tree_Apollo_C|GetTree%20Color%20Option _asparent_tree_apollo_2) "NewEnumerator0")
                (Variables|/Game/Building/ActorBlueprints/Prop/Parent_Tree_Apollo.Parent_Tree_Apollo_C|SetColor "(R=0.088412,G=0.104167,B=0.015629,A=0.000000)" _asparent_tree_apollo_2)
                (else
                  (Class|ParentTreeApollo|SetSeasonPresetColors _asparent_tree_apollo_2)))
              (Class|ParentTreeApollo|UpdateColor _asparent_tree_apollo_2)
              (Variables|/Game/Building/ActorBlueprints/Prop/Parent_Tree_Apollo.Parent_Tree_Apollo_C|SetPrioritySetbyVolume 0 _asparent_tree_apollo_2))))))
    (:then_1
      (if (not IsGlobalUpate)
        (bind _outactors_2 (Actor|GetAllActorsOfClass "/Game/ArtTools/Procedural/Trees/BP_TreeSeasonColorVolume.BP_TreeSeasonColorVolume_C"))
        (for _array_element_3 _outactors_2
          (CallFunction|UpdateTrees _array_element_3 true))))))

(fn ConstructionScript
  (CallFunction|UpdateTrees false))
